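# Translucent, unlit, two-sided smoke material for Niagara mesh particles on the Lone Wolf backpack.
# Two distorted panning noise layers, weighted by mirrored curve-atlas lookups along U, are remapped to a two-color blend tinted by vertex and particle color.
# Opacity combines noise, gradient and trail masks with view-angle fade, plus depth fade off mobile.
# A panning distortion texture drives vertex offset along normals, or a masked local vector via Use Rune.
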Material /BRCosmetics/Accessories/FORT_Backpacks/Backpack_M_MED_Lone_Wolf/FX/M_Backpack_M_MED_Lone_Wolf_Smoke
graph {
 properties {
  bUsedWithNiagaraMeshParticles=true
  blendMode=BLEND_Translucent
  pixelDepthOffsetMode=PDOM_Legacy
  previewMesh=&/SM_Backpack_M_MED_Lone_Wolf_Plane
  shadingModel=MSM_Unlit
  twoSided=true
  usedShadingModels=Unlit
 }
 nodes [
  n1 = VectorParameter() {group=Texture, parameterName=Noise UV, defaultValue=<rgba 0.85 1.5 -0.15 0>}
  n2 = Multiply(A=TextureCoordinate(), B=AppendVector(A=@n1.R, B=@n1.R))
  n3 = AppendVector(A=Multiply(A=@n1.B) {constB=-1}, B=@n1.A)
  n4 = TextureCoordinate()
  n5 = VectorParameter() {group=Texture, parameterName=Distort UV, defaultValue=<rgba 0.75 3 -0.6 0>}
  n6 = AppendVector(A=@n5.R, B=@n5.R)
  n7 = Multiply(A=@n5.B) {constB=-1}
  n8 = TextureSampleParameter2D(UVs=Panner(Coordinate=Multiply(A=@n4, B=@n6), Speed=AppendVector(A=@n7, B=@n5.A))) {
   group=Texture
   parameterName=Distort Tex
   samplerType=SAMPLERTYPE_Normal
   texture=&T_Weap_Distort_N_02
  }
  n9 = ComponentMask(None=@n8.RGB) {g=true, r=true}
  n10 = ScalarParameter() {defaultValue=0.15, group=Texture, parameterName=Distort Intensity}
  n11 = TextureSampleParameter2D(UVs=Add(A=Panner(Coordinate=@n2, Speed=@n3), B=Multiply(A=@n9, B=@n10) {constB=0.28})) {group=Texture, parameterName=Noise Tex, samplerType=SAMPLERTYPE_Masks, texture=/Game/Weapons/WeaponSkins/Wraps/Textures/Mash/Mash_Noise}
  n12 = Power(Base=@n11.R, Exp=ScalarParameter() {defaultValue=0.7, group=Texture, parameterName=Noise Power}) {constExponent=0.8}
  n13 = ComponentMask(None=TextureCoordinate()) {r=true}
  n14 = CurveAtlasRowParameter(CurveTime=OneMinus(None=@n13)) {
   atlas=/Game/Athena/LinearColorCurves/VFX_CurveAtlas
   curve=/Game/Accessories/FORT_Backpacks/Backpack_M_MED_Lone_Wolf/FX/Curve_Backpack_M_MED_Lone_Wolf
   defaultValue=54
   group=Color
   parameterName=Curve_Color
  }
  n15 = CurveAtlasRowParameter(CurveTime=@n13) {
   atlas=&VFX_CurveAtlas
   curve=&Curve_Backpack_M_MED_Lone_Wolf
   defaultValue=54
   group=Color
   parameterName=Curve_Color
  }
  n16 = VectorParameter() {group=Texture, parameterName=Noise UV, defaultValue=<rgba 0.85 1.5 -0.15 0>}
  n17 = Multiply(A=TextureCoordinate(), B=AppendVector(A=@n16.R, B=@n16.R))
  n18 = AppendVector(A=@n16.B, B=@n16.B)
  n19 = TextureCoordinate()
  n20 = VectorParameter() {group=Texture, parameterName=Distort UV, defaultValue=<rgba 0.75 3 -0.6 0>}
  n21 = AppendVector(A=@n20.R, B=@n20.R)
  n22 = TextureSampleParameter2D(UVs=Panner(Coordinate=Multiply(A=@n19, B=@n21), Speed=AppendVector(A=@n20.B, B=@n20.B))) {
   group=Texture
   parameterName=Distort Tex
   samplerType=SAMPLERTYPE_Normal
   texture=&T_Weap_Distort_N_02
  }
  n23 = ComponentMask(None=@n22.RGB) {g=true, r=true}
  n24 = ScalarParameter() {defaultValue=0.15, group=Texture, parameterName=Distort Intensity}
  n25 = TextureSampleParameter2D(UVs=Add(A=Panner(Coordinate=@n17, Speed=@n18), B=Multiply(A=@n23, B=@n24) {constB=0.28})) {group=Texture, parameterName=Noise Tex, samplerType=SAMPLERTYPE_Masks, texture=&Mash_Noise}
  n26 = Power(Base=@n25.R, Exp=ScalarParameter() {defaultValue=0.7, group=Texture, parameterName=Noise Power}) {constExponent=0.8}
  n27 = Add(A=Multiply(A=Saturate(None=@n12), B=@n14.B), B=Multiply(A=@n15.B, B=Saturate(None=@n26)))
  n28 = TextureCoordinate()
  n29 = ComponentMask(None=@n28) {r=true}
  n30 = Multiply(
   A=Multiply(A=OneMinus(None=@n29), B=ComponentMask(None=@n28) {r=true})
   B=ScalarParameter() {defaultValue=5, group=Color, parameterName=Gradient Mul}
  )
  n31 = ScalarParameter() {defaultValue=2, group=Color, parameterName=Gradient Power}
  n32 = LinearInterpolate(A=@n27, B=Multiply(A=Power(Base=@n30, Exp=@n31), B=@n27), Alpha=ScalarParameter() {group=Color, parameterName=Use Gradient}) {constAlpha=1}
  n33 = TextureCoordinate()
  n34 = ComponentMask(None=@n33) {r=true}
  n35 = Multiply(
   A=Multiply(A=OneMinus(None=@n34), B=ComponentMask(None=@n33) {r=true})
   B=ScalarParameter() {defaultValue=6, group=Color, parameterName=Inside Color Range}
  ) {constB=2}
  n36 = Multiply(A=Saturate(None=OneMinus(None=@n35)), B=ScalarParameter() {defaultValue=2, group=Color, parameterName=Inside Color Intensity})
  n37 = ScalarParameter() {group=Color, parameterName=Use Inside Color}
  n38 = CurveAtlasRowParameter(CurveTime=LinearInterpolate(A=@n32, B=Add(A=@n32, B=@n36), Alpha=@n37)) {
   atlas=&VFX_CurveAtlas
   curve=&Curve_Backpack_M_MED_Lone_Wolf
   defaultValue=54
   group=Color
   parameterName=Curve_Color
  }
  n39 = Desaturation(
   None=LinearInterpolate(
    A=VectorParameter() {group=Color, parameterName=ColorA, defaultValue=<rgba 0 0 0 1>}.RGB
    B=VectorParameter() {group=Color, parameterName=ColorB, defaultValue=<rgba 1 1 1 1>}.RGB
    Alpha=LinearInterpolate(A=@n38.R, B=@n38.R, Alpha=@n37)
   )
   Fraction=ScalarParameter() {defaultValue=-0.5, group=Color, parameterName=Desaturation}
  )
  n40 = VertexColor()
  n41 = ParticleColor()
  MaterialFunctionCall(Emissive=Multiply(A=Multiply(A=@n39, B=@n40), B=@n41.RGB)) {materialFunction=&MF_FX_EmissiveAdjustments}
  n42 = CurveAtlasRowParameter(CurveTime=ComponentMask(None=TextureCoordinate()) {r=true}) {
   atlas=&VFX_CurveAtlas
   curve=&Curve_Backpack_M_MED_Lone_Wolf
   defaultValue=54
   group=Color
   parameterName=Curve_Color
  }
  n43 = TextureCoordinate()
  n44 = Add(
   A=Multiply(
    A=TextureSample() {texture=/Game/Global/Textures/Default/Masks_and_Gradients/CylGradient}.R
    B=ScalarParameter() {defaultValue=3, group=Alpha, parameterName=Alpha Width}
   ) {constB=4}
   B=TextureSample() {texture=/Game/Global/Textures/Default/Masks_and_Gradients/gradient_stasis}.R
  )
  n45 = LinearInterpolate(
   A=TextureSample(UVs=Panner(Coordinate=@n43) {speedX=-0.5}) {texture=/Game/Effects/Fort_Effects/Textures/Smoke/T_Lobber_Trail_Mask}.R
   B=@n44
   Alpha=Constant() {r=0.8}
  )
  n46 = Multiply(
   A=LinearInterpolate(A=@n42.A, B=Constant() {r=1}, Alpha=ScalarParameter() {group=Alpha, parameterName=Use Inside Alpha})
   B=Multiply(A=@n45) {constA=5, constB=0.75}
  ) {constA=1}
  n47 = ScalarParameter() {defaultValue=0.6, group=Alpha, parameterName=Lerp Noise}
  n48 = Multiply(A=Multiply(A=LinearInterpolate(A=@n38.R, B=@n46, Alpha=@n47) {constAlpha=0.6}, B=@n45), B=Saturate(None=Power(Base=@n44) {constExponent=4}))
  n49 = ScalarParameter() {defaultValue=3, group=Alpha, parameterName=FadeDistance}
  n50 = ShadingPathSwitch(Default=Multiply(A=@n48, B=DepthFade(FadeDistance=@n49)), Mobile=@n48) {desc=Migrated from FeatureLevelSwitch}
  n51 = MaterialFunctionCall("FadeSensitivity (optional)"=ScalarParameter() {group=Alpha, parameterName=FadeViewAngle}) {materialFunction=&FadeBasedOnViewAngle}
  Multiply(A=@n40.A, B=Multiply(A=@n41.A, B=Multiply(A=@n50, B=@n51.Result)))
  n52 = TextureCoordinate()
  n53 = Multiply(A=OneMinus(None=ComponentMask(None=@n52) {r=true}), B=ComponentMask(None=@n52) {r=true})
  n54 = LinearInterpolate(
   A=AppendVector(A=Multiply(A=@n53) {constB=2}, B=ComponentMask(None=@n52) {g=true})
   B=@n52
   Alpha=ScalarParameter() {group=WPO, parameterName=Use Rune}
  ) {constAlpha=0}
  n55 = VectorParameter() {group=WPO, parameterName=WPO UV, defaultValue=<rgba 3 2 0.75 -0.1>}
  n56 = AppendVector(A=@n55.R, B=@n55.R)
  n57 = TextureSampleParameter2D(UVs=Panner(Coordinate=Multiply(A=@n54, B=@n56), Speed=AppendVector(A=@n55.B, B=@n55.B))) {
   group=WPO
   parameterName=WPO Tex
   samplerType=SAMPLERTYPE_Normal
   texture=&T_Weap_Distort_N_02
  }
  n58 = Multiply(A=@n57.RGB, B=VertexNormalWS())
  n59 = Multiply(A=@n57.G, B=VectorParameter() {group=WPO, parameterName=WPO_Vector, defaultValue=<rgba 1 0 0 0>}.RGB)
  n60 = TextureCoordinate()
  n61 = Multiply(A=OneMinus(None=ComponentMask(None=@n60) {r=true}), B=ComponentMask(None=@n60) {r=true})
  n62 = ScalarParameter() {defaultValue=6, group=WPO, parameterName=Rune WPO Range}
  n63 = Multiply(
   A=Multiply(A=@n57.G, B=Transform(None=@n59) {transformSourceType=TRANSFORMSOURCE_Local})
   B=Saturate(None=Multiply(A=@n61, B=@n62) {constB=2})
  )
  n64 = ScalarParameter() {group=WPO, parameterName=Use Rune}
  Reroute(None=Multiply(A=LinearInterpolate(A=@n58, B=@n63, Alpha=@n64) {constAlpha=0}, B=ScalarParameter() {defaultValue=2, group=WPO, parameterName=WPO Intensity}))
 ]
}
