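# Niagara effect for the Lone Wolf backpack cosmetic, a persistent red glow built from layered smoke and rune meshes over a sprite named DarkGlow.
# Five burst emitters spawn offset particles that do not expire until the emitter deactivates.
# The Outer, Inside, Dark and Rune mesh layers take their color from User color parameters, dimmed by time of day through FNTimeOfDayScalar.
# All layers use small camera offsets.
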NiagaraSystem /BRCosmetics/Accessories/FORT_Backpacks/Backpack_M_MED_Lone_Wolf/FX/NS_Backpack_M_MED_Lone_Wolf
dependencies {usedDataInterfaces [&NiagaraDataInterfaceColorCurve], usedDynamicInputs [], usedModules [&SystemState]}
emitters {
 DarkGlow {
  emitterUpdateScript [
   &EmitterState(Life Cycle Mode=::NewEnumerator1, Loop Behavior=::NewEnumerator1, Loop Duration Mode=::NewEnumerator1, UseLoopDelay=false)
   &SpawnBurst_Instantaneous(Use Spawn Probability=false, UseLoopCountLimit=false)
  ]
  particleSpawnScript [
   &InitializeParticle(Position Offset=<0 -2 -3>, Sprite Size Mode=::NewEnumerator3, Uniform Sprite Size=43, UsePositionOffset=true, Write Material Random=true)
   /SetVariables(Particles.SpriteFacing=<0 1 0>) {isSetParameters=true}
  ]
  particleUpdateScript [
   &ParticleState(Kill Particles When Lifetime Has Elapsed=false, Let Infinitely Lived Particles Die When Emitter Deactivates=true)
   &CameraOffset(Camera Offset Amount=1)
  ]
  rendererClasses [&NiagaraSpriteRendererProperties]
  renderers [{rendererClass=&NiagaraSpriteRendererProperties}]
 }
 Somke_Mesh_Dark {
  emitterUpdateScript [
   &EmitterState(Life Cycle Mode=::NewEnumerator1, Loop Behavior=::NewEnumerator1, Loop Duration Mode=::NewEnumerator1, UseLoopDelay=false)
   &SpawnBurst_Instantaneous(Use Spawn Probability=false, UseLoopCountLimit=false)
  ]
  particleSpawnScript [&InitializeParticle(
    Mesh Scale=&Multiply_VectorByFloat(Vector=<0.95 0.8 1>, Float=0.5)
    Mesh Scale Mode=::NewEnumerator4
    Position Offset=<0 2 2.5>
    UsePositionOffset=true
    Write Material Random=true
   )]
  particleUpdateScript [
   &ParticleState(Kill Particles When Lifetime Has Elapsed=false, Let Infinitely Lived Particles Die When Emitter Deactivates=true)
   &Color(Color=$User.Color Dark)
   &ScaleColor(
    Color Value To Scale=$Particles.Color
    Scale RGB=&VectorFromFloat(Value=&FNTimeOfDayScalar(Night Scale=0.7, Morning Scale=0.3, Evening Scale=0.5))
    ScaleA=true
    ScaleRGB=true
   )
   &CameraOffset(Camera Offset Amount=1)
  ]
  rendererClasses [&NiagaraMeshRendererProperties]
  renderers [{rendererClass=&NiagaraMeshRendererProperties, rendererIndex=0}]
 }
 Somke_Mesh_Inside {
  emitterUpdateScript [
   &EmitterState(Life Cycle Mode=::NewEnumerator1, Loop Behavior=::NewEnumerator1, Loop Duration Mode=::NewEnumerator1, UseLoopDelay=false)
   &SpawnBurst_Instantaneous(Use Spawn Probability=false, UseLoopCountLimit=false)
  ]
  particleSpawnScript [&InitializeParticle(
    Mesh Scale=&Multiply_VectorByFloat(Vector=<1 1 1>, Float=0.9)
    Mesh Scale Mode=::NewEnumerator4
    Position Offset=<0 4 1>
    UsePositionOffset=true
    Write Material Random=true
   )]
  particleUpdateScript [
   &ParticleState(Kill Particles When Lifetime Has Elapsed=false, Let Infinitely Lived Particles Die When Emitter Deactivates=true)
   &Color(Color=$User.Color Inside)
   &ScaleColor(
    Color Value To Scale=$Particles.Color
    Scale RGB=&VectorFromFloat(Value=&FNTimeOfDayScalar(Night Scale=0.7, Morning Scale=0.3, Evening Scale=0.5))
    ScaleA=true
    ScaleRGB=true
   )
   &CameraOffset(Camera Offset Amount=2)
  ]
  rendererClasses [&NiagaraMeshRendererProperties]
  renderers [{rendererClass=&NiagaraMeshRendererProperties, rendererIndex=0}]
 }
 Somke_Mesh_Outer {
  emitterUpdateScript [
   &EmitterState(Life Cycle Mode=::NewEnumerator1, Loop Behavior=::NewEnumerator1, Loop Duration Mode=::NewEnumerator1, UseLoopDelay=false)
   &SpawnBurst_Instantaneous(Use Spawn Probability=false, UseLoopCountLimit=false)
  ]
  particleSpawnScript [&InitializeParticle(
    Mesh Scale=&Multiply_VectorByFloat(Vector=<1 0.5 1>, Float=1)
    Mesh Scale Mode=::NewEnumerator4
    Position Offset=<0 4 1>
    UsePositionOffset=true
    Write Material Random=true
   )]
  particleUpdateScript [
   &ParticleState(Kill Particles When Lifetime Has Elapsed=false, Let Infinitely Lived Particles Die When Emitter Deactivates=true)
   &Color(Color=$User.Color Outer)
   &ScaleColor(
    Color Value To Scale=$Particles.Color
    Scale RGB=&VectorFromFloat(Value=&FNTimeOfDayScalar(Night Scale=0.7, Morning Scale=0.3, Evening Scale=0.5))
    ScaleA=true
    ScaleRGB=true
   )
   &CameraOffset(Camera Offset Amount=2)
  ]
  rendererClasses [&NiagaraMeshRendererProperties]
  renderers [{rendererClass=&NiagaraMeshRendererProperties, rendererIndex=0}]
 }
 Somke_Mesh_Rune {
  emitterUpdateScript [
   &EmitterState(Life Cycle Mode=::NewEnumerator1, Loop Behavior=::NewEnumerator1, Loop Duration Mode=::NewEnumerator1, UseLoopDelay=false)
   &SpawnBurst_Instantaneous(Use Spawn Probability=false, UseLoopCountLimit=false)
  ]
  particleSpawnScript [&InitializeParticle(
    Color=$User.Color Rune
    Mesh Scale=<0.85 2 0.95>
    Mesh Scale Mode=::NewEnumerator4
    Position Offset=<0.5 -2 -6>
    UsePositionOffset=true
    Write Material Random=true
   )]
  particleUpdateScript [
   &ParticleState(Kill Particles When Lifetime Has Elapsed=false, Let Infinitely Lived Particles Die When Emitter Deactivates=true)
   &Color(Color=&MultiplyLinearColor(Linear Color=$User.Color Rune, Scale Factor=&MakeVector4_FromFloat(Float=0.8)))
   &ScaleColor(
    Color Value To Scale=$Particles.Color
    Scale RGB=&VectorFromFloat(Value=&FNTimeOfDayScalar(Night Scale=0.7, Morning Scale=0.3, Evening Scale=0.5))
    ScaleA=true
    ScaleRGB=true
   )
   &CameraOffset(Camera Offset Amount=2)
  ]
  rendererClasses [&NiagaraMeshRendererProperties]
  renderers [{rendererClass=&NiagaraMeshRendererProperties, rendererIndex=0}]
 }
}
summary {
 systemName=NS_Backpack_M_MED_Lone_Wolf
 emitters [
  {emitterName=DarkGlow, rendererClasses [&NiagaraSpriteRendererProperties]}
  {emitterName=Somke_Mesh_Outer, rendererClasses [&NiagaraMeshRendererProperties]}
  {emitterName=Somke_Mesh_Inside, rendererClasses [&NiagaraMeshRendererProperties]}
  {emitterName=Somke_Mesh_Rune, rendererClasses [&NiagaraMeshRendererProperties]}
  {emitterName=Somke_Mesh_Dark, rendererClasses [&NiagaraMeshRendererProperties]}
 ]
 userVariables [
  User.Color Dark {type=LinearColor, description="", defaultValue {struct=LinearColor, value=<rgba 0.8 0 0 1>}}
  User.Color Inside {type=LinearColor, description="", defaultValue {struct=LinearColor, value=<rgba 1.5 0.025 0 1>}}
  User.Color Outer {type=LinearColor, description="", defaultValue {struct=LinearColor, value=<rgba 1 0.0167 0 1>}}
  User.Color Rune {type=LinearColor, description="", defaultValue {struct=LinearColor, value=<rgba 1.3 0.0433 0 1>}}
 ]
}
